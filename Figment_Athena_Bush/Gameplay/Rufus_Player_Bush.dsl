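; Player bush actor that starts hidden in game, takes its instigator's team and is destroyed by damage or on request.
; Server damage resolves the causer's owner, looking through weapons, and counts only hits from the instigator or non-teammates.
; Reaching MaxHealth or any self-inflicted hit, like RemoveBush, sends the Bush.Destroyed gameplay event to the instigator and destroys the actor.
; Server death disables collision on the bush mesh.
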(event EventBeginPlay
  (bind _returnvalue (Game|GetInstigator))
  ;[1 1 1] EVENT -> Begin Play
  (Rendering|SetActorHiddenInGame true)
  (if (Utilities|IsValid _returnvalue :!class /Script/Engine.KismetSystemLibrary)
    (CallFunction|SetTeam (Team|GetActorTeam _returnvalue))))

(event Custom|RemoveBush
  ;[1 1 1] EVENT -> On Damage - Destroy actor if enough damage has been taken. Ensure that the damage source is either an enemy or the owning player
  (Ability|SendGameplayEventtoActor (Variables|Actor|GetInstigator) "(TagName=\"Abilities.Generic.Athena.Bush.Destroyed\")")
  (Actor|DestroyActor))

(event EventOnDamageServer
  (params (Damage :type Float%20%28single-precision%29) (DamageTags :type Gameplay%20Tag%20Container%20Structure%20%28by%20ref%29) (Momentum :type Vector :default "0, 0, 0") (HitInfo :type Hit%20Result%20Structure%20%28by%20ref%29) (InstigatedBy :type Controller%20Object%20Reference) (DamageCauser :type Actor%20Object%20Reference) (EffectContext :type Gameplay%20Effect%20Context%20Handle%20Structure))
  (bind _instigator (Variables|Actor|GetInstigator))
  ;[1 1 1] EVENT -> On Damage - Destroy actor if enough damage has been taken. Ensure that the damage source is either an enemy or the owning player
  (bind _owner (CallFunction|CheckInstIsOwner DamageCauser))
  (if (or (Utilities|Equal(Object) _owner _instigator) (not (Team|OnSameTeam DamageCauser _instigator)))
    (bind _output_get (Variables|Default|SetTotalDamage (+ Damage (Variables|Default|GetTotalDamage))))
    (if (or (>= _output_get (ScalableFloat|GetValueatLevel (Variables|Default|GetMaxHealth))) (Utilities|Equal(Object) _owner _instigator))
      (Ability|SendGameplayEventtoActor (Variables|Actor|GetInstigator) "(TagName=\"Abilities.Generic.Athena.Bush.Destroyed\")")
      (Actor|DestroyActor))))

(event EventOnDeathServer
  (params (Damage :type Float%20%28single-precision%29) (DamageTags :type Gameplay%20Tag%20Container%20Structure%20%28by%20ref%29) (Momentum :type Vector :default "0, 0, 0") (HitInfo :type Hit%20Result%20Structure%20%28by%20ref%29) (InstigatedBy :type Controller%20Object%20Reference) (DamageCauser :type Actor%20Object%20Reference) (EffectContext :type Gameplay%20Effect%20Context%20Handle%20Structure))
  ;[1 1 1] EVENT -> On Death
  (Collision|SetCollisionEnabled (Variables|Default|GetSM_Bush) "NoCollision"))

(fn CheckInstIsOwner
  (params (DamageCauser :type Actor%20Object%20Reference))
  (bind _returnvalue (Actor|GetOwner DamageCauser))
  (bind (_asfort_weapon _bsuccess) (Utilities|Casting|CastToFortWeapon _returnvalue))
  ;[1 1 1] Return the owner if valid from the damage causer.
  (Utilities|IsValid DamageCauser
    (:Is%20Valid
      (Utilities|IsValid _returnvalue
        (:Is%20Valid
          (return (select _bsuccess (Actor|GetOwner _asfort_weapon) _returnvalue)))
        (:Is%20Not%20Valid
          _)))
    (:Is%20Not%20Valid
      _))
  (return DamageCauser))

(fn ConstructionScript
  (CallFunction|SetTeam (Team|GetActorTeam (Game|GetInstigator))))
